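; Gameplay ability for the knockback applied by the Rift Gauntlets Ascend dash, extending a parent knockback ability.
; On activation it runs the parent logic, then gets the avatar as a valid Fort player pawn.
; That pawn briefly ignores collision with the launcher actor from the parent's variables, so the knockback moves the player as intended.
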(event Ability|EventActivateAbility
  ;[0.01328 0.15 0.07985 0.5] ActivateAbility: Ignore collision with our launcher for a short duration to allow the knockback to move the player as intended.
  (|CallParentFunction)
  (bind _outfortplayerpawn (CallFunction|GetAvatarasValidFortPlayerPawn)
    (:ValidFortPlayerPawn
      (Collision|AddIgnoredActor _outfortplayerpawn (Variables|KnockbackParent|GetLauncher) 0.3))))
